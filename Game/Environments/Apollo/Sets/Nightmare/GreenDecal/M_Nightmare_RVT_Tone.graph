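# Two-tone cliff-textured ground-patch material for the Apollo Nightmare set that writes into a Runtime Virtual Texture with a soft, broken-up edge.
# The diffuse is tinted by two colors height-blended on its alpha, then overlaid with a rock detail texture tiled by object radius that also modulates specular and roughness.
# An exponential radial gradient, eroded by the same heightmap, drives opacity.
# A virtual-texture feature switch routes through EnableRVT or falls back to direct attributes.
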Material /Game/Environments/Apollo/Sets/Nightmare/GreenDecal/M_Nightmare_RVT_Tone
graph {
 properties {bTangentSpaceNormal=false, bUseMaterialAttributes=true, pixelDepthOffsetMode=PDOM_Legacy, usedShadingModels=Default Lit}
 nodes [
  n1 = Multiply(A=TextureCoordinate(), B=ScalarParameter() {defaultValue=2, group=Textures, parameterName=ScaleTexture, sortPriority=2})
  n2 = TextureSampleParameter2D(UVs=@n1) {group=Textures, parameterName=Diffuse, sortPriority=0, texture=/Game/Environments/Apollo/Sets/Terrain/Textures/T_Apollo_Cliff_D_02}
  n3 = VectorParameter() {group=Textures, parameterName=DiffuseTone2, defaultValue=<rgba 0.8177 0.7389 0.5302 0>}
  n4 = VectorParameter() {group=Textures, parameterName=DiffuseTone, defaultValue=<rgba 0.7113 0.8984 0.5194 0>}
  n5 = TextureSampleParameter2D(UVs=@n1) {
   group=Textures
   parameterName=Specular
   samplerType=SAMPLERTYPE_Masks
   sortPriority=0
   texture=/Game/Environments/Apollo/Sets/Terrain/Textures/T_Apollo_Cliff_S_02
  }
  n6 = TextureSampleParameter2D(UVs=@n1) {
   group=Textures
   parameterName=Normal
   samplerType=SAMPLERTYPE_Normal
   sortPriority=0
   texture=/Game/Environments/Apollo/Sets/Terrain/Textures/T_Apollo_Cliff_N_02
  }
  n7 = SetMaterialAttributes!(
   MaterialAttributes
   Base Color=MaterialFunctionCall(
    A=Multiply(A=@n2.RGB, B=@n3.RGB)
    B=Multiply(A=@n2.RGB, B=@n4.RGB)
    Transition Phase=ScalarParameter() {defaultValue=0.5, group=Textures, parameterName=ColorTransition, sortPriority=3}
    Height Texture=@n2.A
    Contrast=ScalarParameter() {defaultValue=1, group=Textures, parameterName=ColorTransitionContrast, sortPriority=4}
   ) {materialFunction=&HeightLerp}.Results
   Roughness=Multiply(A=@n5.B, B=ScalarParameter() {defaultValue=0.35, group=Textures, parameterName=Roughness, sortPriority=2})
   Specular=Multiply(A=@n5.R, B=ScalarParameter() {defaultValue=0.5, group=Textures, parameterName=Specular, sortPriority=1})
   Normal=Transform(None=@n6.RGB)
  ) {attributeSetTypes [
    #1
    #2
    #3
    #4
   ]}
  n8 = GetMaterialAttributes(None=@n7) {attributeGetTypes [#1, #3, #2]}
  n9 = ObjectRadius()
  n10 = ScalarParameter() {defaultValue=300, group=Detail Texture, parameterName=DetailTile, sortPriority=3}
  n11 = TextureSampleParameter2D(UVs=Multiply(A=TextureCoordinate(), B=Divide(A=@n9, B=@n10))) {group=Detail Texture, parameterName=Details, texture=/Game/Environments/Apollo/Rocks/Natural_Rocks/Textures/T_RockDetails_02D}
  n12 = LinearInterpolate(
   A=ScalarParameter() {defaultValue=-0.773, group=Detail Texture, parameterName=Detail R Min, sortPriority=5}
   B=ScalarParameter() {defaultValue=1.8112, group=Detail Texture, parameterName=Detail R Max, sortPriority=6}
   Alpha=@n11.R
  )
  n13 = MaterialFunctionCall(Radius=Constant() {r=0.5}, Density=ScalarParameter() {defaultValue=12, group=Transition, parameterName=MaskDensity, sortPriority=1}) {materialFunction=&RadialGradientExponential}
  n14 = Multiply(A=OneMinus(None=@n2.A), B=ScalarParameter() {defaultValue=0.5, group=Transition, parameterName=PushHeightmap, sortPriority=3})
  n15 = SetMaterialAttributes!(
   MaterialAttributes=SetMaterialAttributes!(
    MaterialAttributes=@n8.MaterialAttributes
    Base Color=MaterialFunctionCall(Base=@n8.Base Color, Blend=@n11.RGB) {materialFunction=&Blend_Overlay}.Result
    Specular=Multiply(A=@n8.Specular, B=@n11.R)
    Roughness=Multiply(A=@n8.Roughness, B=@n12)
   ) {attributeSetTypes [#1, #3, #2]}
   Opacity=MaterialFunctionCall(
    Transition Phase=Subtract(A=@n13.RadialGradientExponential, B=@n14)
    Height Texture=@n2.A
    Contrast=ScalarParameter() {defaultValue=1, group=Transition, parameterName=EdgeContrast, sortPriority=2}
   ) {materialFunction=&HeightLerp}.Alpha
  ) {attributeSetTypes [#5]}
  n16 = GetMaterialAttributes(None=@n15) {attributeGetTypes [
    #1
    #3
    #2
    #4
    #5
   ]}
  RuntimeVirtualTextureOutput!(
   BaseColor=@n16.Base Color
   Specular=@n16.Base Color
   Roughness=@n16.Base Color
   Normal=@n16.Base Color
   WorldHeight
   Opacity=@n16.Base Color
   Mask
   Displacement
   Mask4
  )
  n17 = VirtualTextureFeatureSwitch!(
   No=@n15
   Yes=MaterialFunctionCall!(RVT=@n16.MaterialAttributes, ByPass RVT=@n15) {materialFunction=/Game/Athena/Apollo/Maps/Landscape/VT/MaterialFunctions/EnableRVT}.Result
  )
  ShadingPathSwitch(Default=@n17, Deferred=@n17, Forward=@n17, Mobile=@n17)
 ]
}
